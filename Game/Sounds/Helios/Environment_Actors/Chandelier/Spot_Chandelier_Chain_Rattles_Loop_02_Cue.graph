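# Looping spot ambience cue for chandelier chain rattles, routed to the earthquakes sound class and exterior ambient source bus.
# A loop with a random 6 to 12 second delay mixes enveloped, modulated random slow chain rattles with half-volume random chandelier rattles.
# Each layer has its own attenuation node with listener focus and reverb sends.
# Only the first two quality levels connect the loop; the other two are empty.
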SoundCue /Game/Sounds/Helios/Environment_Actors/Chandelier/Spot_Chandelier_Chain_Rattles_Loop_02_Cue
root=QualityLevel_0
nodes [
 {
  id=Attenuation_0
  children [Modulator_1]
  properties {
   attenuationSettings=&/Spot_Chandelier_With_Heavy_Focus_Att
   attenuationOverrides {
    bEnableFocusInterpolation=true
    bEnableListenerFocus=true
    bEnableSubmixSends=true
    distanceAlgorithm=Custom
    falloffDistance=1500
    focusReleaseInterpSpeed=0.25
    focusVolumeAttenuation=1.1
    nonFocusVolumeAttenuation=0.5
    attenuationShapeExtents {x=0}
    customAttenuationCurve {editorCurveData {keys [<at 0.0051 0.4771 RCIM_Cubic>, {arriveTangent=-0.4796, interpMode=RCIM_Cubic, leaveTangent=-0.4796, time=0.4, value=1}, <at 1 0>]}}
    submixSendSettings [{
      bEnableHPFCutoff=false
      bEnableLPFCutoff=false
      disableManualSendClamp=false
      hPFCutoff=20
      lPFCutoff=20000
      maxSendDistance=6000
      maxSendLevel=1
      minSendDistance=400
      minSendLevel=0
      sendLevel=1
      sendLevelControlMethod=Linear
      soundSubmix=&Fortnite_Reverb_Default
     }]
   }
  }
 }
 {
  id=Attenuation_1
  children [Modulator_0]
  properties {
   bOverrideAttenuation=true
   attenuationOverrides {
    bEnableFocusInterpolation=true
    bEnableListenerFocus=true
    bEnableSubmixSends=true
    distanceAlgorithm=Custom
    falloffDistance=1500
    focusReleaseInterpSpeed=0.25
    focusVolumeAttenuation=1.1
    nonFocusVolumeAttenuation=0.5
    attenuationShapeExtents {x=0}
    customAttenuationCurve {editorCurveData {keys [<at 0.0051 0.4771 RCIM_Cubic>, {arriveTangent=-0.4796, interpMode=RCIM_Cubic, leaveTangent=-0.4796, time=0.4, value=1}, <at 1 0>]}}
    submixSendSettings [
     {
      bEnableHPFCutoff=false
      bEnableLPFCutoff=false
      disableManualSendClamp=false
      hPFCutoff=20
      lPFCutoff=20000
      maxSendDistance=1500
      maxSendLevel=1
      minSendDistance=500
      minSendLevel=0.2
      sendLevel=1
      sendLevelControlMethod=Linear
      soundSubmix=&Fortnite_Reverb_Default
      customSendLevelCurve {editorCurveData {keys [<at 0 0.2>,<at 1 1>]}}
     }
     {
      bEnableHPFCutoff=false
      bEnableLPFCutoff=false
      disableManualSendClamp=false
      hPFCutoff=20
      lPFCutoff=20000
      maxSendDistance=1500
      maxSendLevel=0.4
      minSendDistance=400
      minSendLevel=0
      sendLevel=0.2
      sendLevelControlMethod=Linear
      soundSubmix=&Fortnite_Reverb_Distant
      customSendLevelCurve {externalCurve=/Game/Sounds/SubMixes/Reverb/Fortnite_Reverb_Distant_Send}
     }
    ]
   }
  }
 }
 {id=Delay_0, children [Mixer_0], properties {delayMax=12, delayMin=6}}
 {
  id=Enveloper_0
  children [Random_0]
  properties {pitchCurve {editorCurveData {keys [{time=0}]}}, volumeCurve {editorCurveData {keys [<at 0 1>,<at 3 1>,<at 4 0>]}}}
 }
 {id=Looping_0, children [Delay_0]}
 {id=Mixer_0, children [Attenuation_1, Attenuation_0], properties {inputVolume [1,0.5]}}
 {id=Modulator_0, children [Enveloper_0], properties {pitchMin=0.8, volumeMin=0.8}}
 {id=Modulator_1, children [Random_1], properties {pitchMax=1, pitchMin=1}}
 {id=QualityLevel_0, children [Looping_0,Looping_0,null,null]}
 {
  id=Random_0
  children [WavePlayer_0,WavePlayer_2,WavePlayer_3,WavePlayer_4,WavePlayer_5,WavePlayer_6,WavePlayer_7,WavePlayer_8]
  properties {bShouldExcludeFromBranchCulling=true, weights [1,1,1,1,1,1,1,1]}
 }
 {
  id=Random_1
  children [WavePlayer_1, WavePlayer_9, WavePlayer_10, WavePlayer_11, WavePlayer_12, WavePlayer_13]
  properties {weights [1,1,1,1,1,1]}
 }
 {id=WavePlayer_0, properties {soundWaveAssetPtr=/Game/Sounds/Fort_Starlight/S1/SoundWaves/SpotFX_ChainRattle_Slow_01}}
 {id=WavePlayer_1, properties {soundWaveAssetPtr=&/Spot_Nitro_Wasteland_Chandelier_Rattle_1}}
 {id=WavePlayer_10, properties {soundWaveAssetPtr=&/Spot_Nitro_Wasteland_Chandelier_Rattle_3}}
 {id=WavePlayer_11, properties {soundWaveAssetPtr=&/Spot_Nitro_Wasteland_Chandelier_Rattle_4}}
 {id=WavePlayer_12, properties {soundWaveAssetPtr=&/Spot_Nitro_Wasteland_Chandelier_Rattle_5}}
 {id=WavePlayer_13, properties {soundWaveAssetPtr=&/Spot_Nitro_Wasteland_Chandelier_Rattle_6}}
 {id=WavePlayer_2, properties {soundWaveAssetPtr=/Game/Sounds/Fort_Starlight/S1/SoundWaves/SpotFX_ChainRattle_Slow_02}}
 {id=WavePlayer_3, properties {soundWaveAssetPtr=/Game/Sounds/Fort_Starlight/S1/SoundWaves/SpotFX_ChainRattle_Slow_03}}
 {id=WavePlayer_4, properties {soundWaveAssetPtr=/Game/Sounds/Fort_Starlight/S1/SoundWaves/SpotFX_ChainRattle_Slow_04}}
 {id=WavePlayer_5, properties {soundWaveAssetPtr=/Game/Sounds/Fort_Starlight/S1/SoundWaves/SpotFX_ChainRattle_Slow_05}}
 {id=WavePlayer_6, properties {soundWaveAssetPtr=/Game/Sounds/Fort_Starlight/S1/SoundWaves/SpotFX_ChainRattle_Slow_06}}
 {id=WavePlayer_7, properties {soundWaveAssetPtr=/Game/Sounds/Fort_Starlight/S1/SoundWaves/SpotFX_ChainRattle_Slow_07}}
 {id=WavePlayer_8, properties {soundWaveAssetPtr=/Game/Sounds/Fort_Starlight/S1/SoundWaves/SpotFX_ChainRattle_Slow_08}}
 {id=WavePlayer_9, properties {soundWaveAssetPtr=&/Spot_Nitro_Wasteland_Chandelier_Rattle_2}}
]
properties {
 maxDistance=1850
 soundClassObject=/Game/Sounds/FortSoundClassesAndModes/Fort_Earthquakes
 volumeMultiplier=0.4
 attenuationOverrides {
  bAttenuate=false
  bAttenuateWithLPF=true
  bEnableOcclusion=true
  distanceAlgorithm=Custom
  falloffDistance=2500
  lPFFrequencyAtMax=1000
  lPFRadiusMax=3000
  lPFRadiusMin=1000
  occlusionInterpolationTime=3
  occlusionLowPassFilterFrequency=1000
  occlusionTraceChannel=ECC_GameTraceChannel11
  occlusionVolumeAttenuation=0.5
  stereoSpread=100
  attenuationShapeExtents {x=50}
  customAttenuationCurve {editorCurveData {keys [
     <at 0 0>
     {arriveTangent=2.5, interpMode=RCIM_Cubic, leaveTangent=2.5, time=0.25, value=0.0018}
     <at 0.4 1>
     {arriveTangent=-0.0071, interpMode=RCIM_Cubic, leaveTangent=-0.0071, tangentMode=RCTM_User, time=0.6, value=1}
     <at 1 0>
    ]}}
 }
 busSends [{audioBus=null, minSendLevel=1, sendLevel=1, soundSourceBus=&AmbientAudio_Exterior_SourceBus}]
}
