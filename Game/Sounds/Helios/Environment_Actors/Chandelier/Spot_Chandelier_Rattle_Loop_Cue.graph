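# Looping ambient Sound Cue that plays intermittent chandelier rattles for an environment actor.
# A Looping node repeats a random 4 to 12 second Delay, then one of six equally weighted rattle waves.
# Each wave passes through a fixed-pitch Modulator and an Enveloper whose looping curve gently dips and restores volume.
# A heavy-focus attenuation asset handles spatialization, and overridden concurrency caps simultaneous instances per owner.
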SoundCue /Game/Sounds/Helios/Environment_Actors/Chandelier/Spot_Chandelier_Rattle_Loop_Cue
root=Looping_0
nodes [
 {id=Delay_0, children [Enveloper_3], properties {delayMax=12, delayMin=4}}
 {
  id=Enveloper_3
  children [Modulator_0]
  properties {
   bLoop=true
   bLoopIndefinitely=true
   loopEnd=20
   pitchCurve {editorCurveData {keys [{time=0}]}}
   volumeCurve {editorCurveData {keys [
      {arriveTangent=-0.01, interpMode=RCIM_Cubic, leaveTangent=-0.01, time=0, value=1}
      <at 10 0.9 RCIM_Cubic>
      {arriveTangent=0.01, interpMode=RCIM_Cubic, leaveTangent=0.01, time=20, value=1}
     ]}}
  }
 }
 {id=Looping_0, children [Delay_0]}
 {id=Modulator_0, children [Random_0], properties {pitchMax=1, pitchMin=1}}
 {id=Random_0, children [WavePlayer_0,WavePlayer_1,WavePlayer_3,WavePlayer_4,WavePlayer_5,WavePlayer_6], properties {weights [1,1,1,1,1,1]}}
 {id=WavePlayer_0, properties {soundWaveAssetPtr=&/Spot_Nitro_Wasteland_Chandelier_Rattle_1}}
 {id=WavePlayer_1, properties {soundWaveAssetPtr=&/Spot_Nitro_Wasteland_Chandelier_Rattle_2}}
 {id=WavePlayer_3, properties {soundWaveAssetPtr=&/Spot_Nitro_Wasteland_Chandelier_Rattle_3}}
 {id=WavePlayer_4, properties {soundWaveAssetPtr=&/Spot_Nitro_Wasteland_Chandelier_Rattle_4}}
 {id=WavePlayer_5, properties {soundWaveAssetPtr=&/Spot_Nitro_Wasteland_Chandelier_Rattle_5}}
 {id=WavePlayer_6, properties {soundWaveAssetPtr=&/Spot_Nitro_Wasteland_Chandelier_Rattle_6}}
]
properties {
 attenuationSettings=&/Spot_Chandelier_With_Heavy_Focus_Att
 bOverrideConcurrency=true
 maxDistance=1850
 soundClassObject=&Fort_Ambient_Athena
 subtitlePriority=5000
 volumeMultiplier=0.301
 attenuationOverrides {
  bEnableFocusInterpolation=true
  bEnableListenerFocus=true
  bEnableOcclusion=true
  dBAttenuationAtMax=-40
  falloffDistance=512
  focusAzimuth=15
  focusVolumeAttenuation=1.5
  nonFocusAzimuth=45
  nonFocusVolumeAttenuation=0.5
  occlusionInterpolationTime=1
  occlusionLowPassFilterFrequency=10000
  occlusionTraceChannel=ECC_GameTraceChannel11
  occlusionVolumeAttenuation=0.5
  stereoSpread=512
  attenuationShapeExtents=<256 64.5942 90>
 }
 concurrencyOverrides {bLimitToOwner=true, maxCount=3}
}
